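# Ragdoll physics asset for the green BirdNest backpack cosmetic, covering a small humanoid figure rig on the pack.
# Each dynamic bone has one capsule body: head, body, pelvis, arms, legs and two two-segment fingers per hand.
# Angle-limited constraints chain head, shoulders and pelvis to the body, with legs hanging off the pelvis.
# Fingers bend on one swing axis only, and left and right limb limits differ slightly.
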PhysicsAsset /BRCosmetics/Accessories/FORT_Backpacks/Backpack_M_MED_BirdNestGreen/Meshes/M_MED_BirdNestGreen_Pack_Physics
bodies {
 Dyn_head [{shapeType=Capsule, length=0.5542, radius=4.956, center=<4.2354 -0.0877 -2.136>, rotation=<rot 39.9745 1.2053 -177.5745>}]
 Dyn_pelvis [{shapeType=Capsule, length=1.7532, radius=2.5524, center=<-2.9571 0 0.3806>, rotation=<rot 3.7192 180 -90>}]
 dyn_Finger_index_r_end [{
   shapeType=Capsule
   length=1.3157
   radius=0.9636
   shapeName=dyn_Finger_index_r_end_capsule
   center=<-1.2619 -0.2386 -0.2278>
   rotation=<rot 0.704 132.0088 -79.8159>
  }]
 dyn_Finger_index_r_start [{
   shapeType=Capsule
   length=0.5042
   radius=0.8653
   shapeName=dyn_Finger_index_r_start_capsule
   center=<-0.4401 0.0837 -0.0723>
   rotation=<rot 4.2921 -89.0833 -86.1909>
  }]
 dyn_body [{
   shapeType=Capsule
   length=3.018
   radius=4.6731
   shapeName=dyn_body_capsule
   center=<-0.391 0 -0.2998>
   rotation=<rot 81.478 179.9999 -0.0001>
  }]
 dyn_elbow_l [{
   shapeType=Capsule
   length=3.1903
   radius=1.0437
   shapeName=Dyn_elbow_l_capsule
   center=<2.2773 0.0526 0.0009>
   rotation=<rot 0 89.9884 -88.3801>
  }]
 dyn_elbow_r [{
   shapeType=Capsule
   length=3.1903
   radius=1.0437
   shapeName=Dyn_elbow_l_capsule
   center=<-2.2773 -0.0526 -0.0009>
   rotation=<rot 0 89.9884 91.6199>
  }]
 dyn_finger_index_l_end [{
   shapeType=Capsule
   length=1.3157
   radius=0.9636
   shapeName=dyn_Finger_index_r_end_capsule
   center=<1.2619 0.2386 0.2278>
   rotation=<rot 0.704 132.0088 100.1841>
  }]
 dyn_finger_index_l_start [{
   shapeType=Capsule
   length=0.5042
   radius=0.8653
   shapeName=dyn_Finger_index_r_start_capsule
   center=<0.4401 -0.0837 0.0723>
   rotation=<rot 4.2921 -89.0833 93.8091>
  }]
 dyn_finger_pinky_l_end [{
   shapeType=Capsule
   length=1.0431
   radius=1.0433
   shapeName=dyn_finger_pinky_r_end_capsule
   center=<1.3083 0.0163 0.0081>
   rotation=<rot 0 -172.9209 98.1519>
  }]
 dyn_finger_pinky_l_start [{
   shapeType=Capsule
   length=0.6337
   radius=0.7794
   shapeName=dyn_finger_pinky_r_start_capsule
   center=<0.9312 -0.0108 -0.0077>
   rotation=<rot 1.2915 -92.4591 97.1762>
  }]
 dyn_finger_pinky_r_end [{
   shapeType=Capsule
   length=1.0431
   radius=1.0433
   shapeName=dyn_finger_pinky_r_end_capsule
   center=<-1.3083 -0.0163 -0.0081>
   rotation=<rot 0 -172.9209 -81.8481>
  }]
 dyn_finger_pinky_r_start [{
   shapeType=Capsule
   length=0.6337
   radius=0.7794
   shapeName=dyn_finger_pinky_r_start_capsule
   center=<-0.9312 0.0108 0.0077>
   rotation=<rot 1.2915 -92.4591 -82.8238>
  }]
 dyn_foot_l [{
   shapeType=Capsule
   length=3.4646
   radius=1.7037
   shapeName=Dyn_foot_l_capsule
   center=<1.956 -0.1238 2.0466>
   rotation=<rot -11.7741 -86.7165 31.4492>
  }]
 dyn_foot_r [{
   shapeType=Capsule
   length=3.4646
   radius=1.7037
   shapeName=Dyn_foot_l_capsule
   center=<-1.956 0.1238 -2.0466>
   rotation=<rot -11.7741 -86.7165 -148.5508>
  }]
 dyn_knee_l [{
   shapeType=Capsule
   length=4.3315
   radius=1.1246
   shapeName=Dyn_knee_l_capsule
   center=<3.1497 -0.0965 -0.2501>
   rotation=<rot 0 92.8337 -86.5378>
  }]
 dyn_knee_r [{
   shapeType=Capsule
   length=4.3315
   radius=1.1246
   shapeName=Dyn_knee_l_capsule
   center=<-3.1497 0.0965 0.2501>
   rotation=<rot 0 92.8337 93.4622>
  }]
 dyn_shoulder_l [{
   shapeType=Capsule
   length=3.2939
   radius=1.0782
   shapeName=dyn_shoulder_l_capsule
   center=<3.2438 0.0793 0.0205>
   rotation=<rot 0 -87.3289 -87.3097>
  }]
 dyn_shoulder_r [{
   shapeType=Capsule
   length=3.2939
   radius=1.0782
   shapeName=dyn_shoulder_l_capsule
   center=<-3.2438 -0.0793 -0.0205>
   rotation=<rot 0 -87.3289 92.6903>
  }]
 dyn_thing_l [{
   shapeType=Capsule
   length=2.9778
   radius=1.6759
   shapeName=Dyn_thing_l_capsule
   center=<3.676 -0.1036 0.3187>
   rotation=<rot 0.4688 89.2769 -83.2022>
  }]
 dyn_thing_r [{
   shapeType=Capsule
   length=2.9778
   radius=1.6759
   shapeName=Dyn_thing_l_capsule
   center=<-3.676 0.1036 -0.3187>
   rotation=<rot 0.4688 89.2769 96.7978>
  }]
 dyn_wrist_l [{
   shapeType=Capsule
   length=0.7517
   radius=1.265
   shapeName=Dyn_wrist_l_capsule
   center=<0.4089 0.0682 0.0087>
   rotation=<rot 16.7656 93.3925 -99.4672>
  }]
 dyn_wrist_r [{
   shapeType=Capsule
   length=0.7517
   radius=1.265
   shapeName=Dyn_wrist_l_capsule
   center=<-0.4089 -0.0682 -0.0087>
   rotation=<rot 16.7656 93.3925 80.5328>
  }]
}
constraints [
 {bone1Name=Dyn_head, bone2Name=dyn_body, swing1LimitDegrees=35, swing2LimitDegrees=35, twistLimitDegrees=20}
 {bone1Name=Dyn_pelvis, bone2Name=dyn_body, swing1LimitDegrees=20, swing2LimitDegrees=9, twistLimitDegrees=5}
 {bone1Name=dyn_finger_pinky_r_end, bone2Name=dyn_finger_pinky_r_start, swing1LimitDegrees=0, swing2LimitDegrees=20, twistLimitDegrees=0}
 {bone1Name=dyn_finger_pinky_r_start, bone2Name=dyn_wrist_r, swing1LimitDegrees=0, swing2LimitDegrees=22, twistLimitDegrees=0}
 {bone1Name=dyn_Finger_index_r_end, bone2Name=dyn_Finger_index_r_start, swing1LimitDegrees=0, swing2LimitDegrees=20, twistLimitDegrees=0}
 {bone1Name=dyn_Finger_index_r_start, bone2Name=dyn_wrist_r, swing1LimitDegrees=0, swing2LimitDegrees=20, twistLimitDegrees=0}
 {bone1Name=dyn_finger_pinky_l_start, bone2Name=dyn_wrist_l, swing1LimitDegrees=0, swing2LimitDegrees=25, twistLimitDegrees=0}
 {bone1Name=dyn_finger_pinky_l_end, bone2Name=dyn_finger_pinky_l_start, swing1LimitDegrees=0, swing2LimitDegrees=20, twistLimitDegrees=0}
 {bone1Name=dyn_finger_index_l_end, bone2Name=dyn_finger_index_l_start, swing1LimitDegrees=0, swing2LimitDegrees=20, twistLimitDegrees=0}
 {bone1Name=dyn_foot_r, bone2Name=dyn_knee_r, swing1LimitDegrees=10, swing2LimitDegrees=15, twistLimitDegrees=0}
 {bone1Name=dyn_knee_r, bone2Name=dyn_thing_r, swing1LimitDegrees=9, swing2LimitDegrees=7, twistLimitDegrees=3}
 {bone1Name=dyn_thing_r, bone2Name=Dyn_pelvis, swing1LimitDegrees=11, swing2LimitDegrees=6, twistLimitDegrees=8}
 {bone1Name=dyn_wrist_r, bone2Name=dyn_elbow_r, swing1LimitDegrees=5, swing2LimitDegrees=15, twistLimitDegrees=5}
 {bone1Name=dyn_elbow_r, bone2Name=dyn_shoulder_r, swing1LimitDegrees=30, swing2LimitDegrees=38, twistLimitDegrees=4}
 {bone1Name=dyn_shoulder_r, bone2Name=dyn_body, swing1LimitDegrees=30, swing2LimitDegrees=33, twistLimitDegrees=5}
 {bone1Name=dyn_finger_index_l_start, bone2Name=dyn_wrist_l, swing1LimitDegrees=0, swing2LimitDegrees=20, twistLimitDegrees=0}
 {bone1Name=dyn_wrist_l, bone2Name=dyn_elbow_l, swing1LimitDegrees=7, swing2LimitDegrees=20, twistLimitDegrees=4}
 {bone1Name=dyn_elbow_l, bone2Name=dyn_shoulder_l, swing1LimitDegrees=20, swing2LimitDegrees=30, twistLimitDegrees=4}
 {bone1Name=dyn_shoulder_l, bone2Name=dyn_body, swing1LimitDegrees=25, swing2LimitDegrees=35, twistLimitDegrees=5}
 {bone1Name=dyn_foot_l, bone2Name=dyn_knee_l, swing1LimitDegrees=10, swing2LimitDegrees=15, twistLimitDegrees=0}
 {bone1Name=dyn_knee_l, bone2Name=dyn_thing_l, swing1LimitDegrees=8, swing2LimitDegrees=6.5, twistLimitDegrees=3}
 {bone1Name=dyn_thing_l, bone2Name=Dyn_pelvis, swing1LimitDegrees=10, swing2LimitDegrees=4, twistLimitDegrees=8}
]
