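# Physics asset for the wings of the BirdNestGreen backpack cosmetic, defining collision bodies and joint limits for its dynamic bones.
# A capsule on dyn_body anchors mirrored left and right chains of three wing bones, shaped as boxes with an extra small capsule on each tip bone.
# Angular limits tighten from root to tip with no twist allowed at the last joint, and differ slightly between sides.
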PhysicsAsset /BRCosmetics/Accessories/FORT_Backpacks/Backpack_M_MED_BirdNestGreen/Meshes/M_MED_BirdNestGreen_Pack_Physics_wings
bodies {
 dyn_body [{shapeType=Capsule, length=6.0911, radius=5.3038, center=<-0.8551 0 -1.1729>, rotation=<rot 0 -90 76.713>}]
 dyn_wing_l_1 [{
   shapeType=Box
   extentX=3.3546
   extentY=1.0919
   extentZ=5.1666
   shapeName=dyn_wing_l_1_box
   center=<2.0373 0 0.7473>
   rotation=<rot -5 0 0.0001>
  }]
 dyn_wing_l_2 [{shapeType=Box, extentX=3.6532, extentY=1.0844, extentZ=7.5, shapeName=dyn_wing_l_2_box, center=<0.6836 0 0.1205>, rotation=<rot 10 0 0.0001>}]
 dyn_wing_l_3 [
  {shapeType=Capsule, length=0.505, radius=0.505, shapeName=dyn_wing_l_3_capsule, rotation=<rot 0 -90 0>}
  {
   shapeType=Box
   extentX=3.8886
   extentY=1.5879
   extentZ=5.9453
   shapeName=dyn_wing_l_3_box
   center=<0.6685 -0.2503 0.0722>
   rotation=<rot 14.9999 0 0.0001>
  }
 ]
 dyn_wing_r_1 [{
   shapeType=Box
   extentX=3.3546
   extentY=1.0919
   extentZ=5.1666
   shapeName=dyn_wing_l_1_box
   center=<-2.0373 0 -0.7473>
   rotation=<rot -5 0 -179.9999>
  }]
 dyn_wing_r_2 [{
   shapeType=Box
   extentX=3.6532
   extentY=1.0844
   extentZ=7.5
   shapeName=dyn_wing_l_2_box
   center=<-0.6836 0 -0.1205>
   rotation=<rot 10 0 -179.9999>
  }]
 dyn_wing_r_3 [
  {shapeType=Capsule, length=0.505, radius=0.505, shapeName=dyn_wing_l_3_capsule, rotation=<rot 0 -90 180>}
  {
   shapeType=Box
   extentX=3.8886
   extentY=1.5879
   extentZ=5.9453
   shapeName=dyn_wing_l_3_box
   center=<-0.6685 0.2503 -0.0722>
   rotation=<rot 14.9999 0 -179.9999>
  }
 ]
}
constraints [
 {bone1Name=dyn_wing_l_1, bone2Name=dyn_body, swing1LimitDegrees=13, swing2LimitDegrees=23, twistLimitDegrees=5}
 {bone1Name=dyn_wing_l_2, bone2Name=dyn_wing_l_1, swing1LimitDegrees=10, swing2LimitDegrees=13, twistLimitDegrees=4}
 {bone1Name=dyn_wing_l_3, bone2Name=dyn_wing_l_2, swing1LimitDegrees=4, swing2LimitDegrees=8, twistLimitDegrees=0}
 {bone1Name=dyn_wing_r_1, bone2Name=dyn_body, swing1LimitDegrees=14, swing2LimitDegrees=20, twistLimitDegrees=5}
 {bone1Name=dyn_wing_r_2, bone2Name=dyn_wing_r_1, swing1LimitDegrees=9, swing2LimitDegrees=16, twistLimitDegrees=4}
 {bone1Name=dyn_wing_r_3, bone2Name=dyn_wing_r_2, swing1LimitDegrees=5, swing2LimitDegrees=9, twistLimitDegrees=0}
]
