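# Animation Blueprint for the BirdNestGreen back bling, adding procedural motion to the pack and its wings.
# A base pose gets Dyn_body offsets (extra while skydiving), then Anim Dynamics, pack and curve-weighted wing RigidBody sims, and an additive idle.
# The event graph retunes both sims' component velocity and acceleration scales by skydive and ground-movement state.
# It also computes a deprecated gravity override that the AnimGraph never reads.
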AnimBlueprint /BRCosmetics/Accessories/FORT_Backpacks/Backpack_M_MED_BirdNestGreen/Meshes/M_MED_BirdNestGreen_Pack_AnimBP
target_skeleton=&/M_MED_BirdNestGreen_Pack_Skeleton
graphs [
 AnimGraph {nodes [
   Input Pose()
   n1 = "Transform (Modify) Bone - Bone: Dyn_body"!(
    ComponentPose=Local To Component(LocalPose=Sequence Player 'BirdNestGreen_BASEPOSE_BACKBLING'() {sequence=/BRCosmetics/Accessories/FORT_Backpacks/Backpack_M_MED_BirdNestGreen/Animation/BirdNestGreen_BASEPOSE_BACKBLING})
    bAlphaBoolEnabled:bool=True
    Translation="4,0,0"
    Rotation="19,0,0"
    Scale="1,1,1"
   ) -> (Pose) {
    alphaInputType=Bool
    rotationMode=BMM_Additive
    rotationSpace=BCS_BoneSpace
    translationMode=BMM_Additive
    translationSpace=BCS_BoneSpace
    alphaBoolBlend {blendOption=Linear}
    boneToModify {boneName=Dyn_body}
    rotation {pitch=19}
    translation {x=4}
   }
   n2 = "Transform (Modify) Bone - Bone: Dyn_body"(
    ComponentPose=@n1
    bAlphaBoolEnabled=Get FullSkyDiving() -> (FullSkyDiving:bool)
    Translation="0.5,0,0"
    Rotation="7.000,-0.000,-0.000"
    Scale="1,1,1"
   ) {
    alphaInputType=Bool
    rotationMode=BMM_Additive
    rotationSpace=BCS_BoneSpace
    translationMode=BMM_Additive
    translationSpace=BCS_BoneSpace
    alphaBoolBlend {blendOption=Linear}
    boneToModify {boneName=Dyn_body}
    rotation=<rot 6.9999 0 0>
    translation {x=0.5}
   }
   n3 = Anim Dynamics - Bone: Dyn_body(ComponentPose=@n2, Alpha=1, ExternalForce="0,0,0") {
    angularDampingOverride=1
    angularSpringConstant=1500
    bAngularSpring=true
    bLinearSpring=true
    bOverrideAngularDamping=true
    bOverrideLinearDamping=true
    bUseGravityOverride=true
    linearDampingOverride=1
    linearSpringConstant=50
    simulationSpace=RootRelative
    alphaBoolBlend {blendOption=Linear}
    boundBone {boneName=Dyn_body}
    gravityOverride {z=-984}
    physicsBodyDefinitions [{
      boundBone {boneName=Dyn_body}
      boxExtents=<10 10 10>
      constraintSetup {angularLimitsMax=<10 0 10>, angularLimitsMin=<-10 0 -10>, angularTarget=<1 0 0>, linearAxesMax=<3 3 0>, linearAxesMin=<-3 -3 0>}
      localJointOffset=<0 0 0>
     }]
   }
   n4 = RigidBody!(
    ComponentPose=@n3
    Alpha
    OverrideWorldGravity:Vector="0,0,-980"
    ExternalForce
    ComponentLinearAccScale:Vector=Get Component Linear Acc Scale_0() -> (Component Linear Acc Scale_0:Vector)
    ComponentLinearVelScale:Vector=Get Component Linear Vel Scale_0() -> (Component Linear Vel Scale_0:Vector)
   ) -> (Pose) {
    bOverrideWorldGravity=true
    overridePhysicsAsset=&/M_MED_BirdNestGreen_Pack_Physics
    alphaBoolBlend {blendOption=Linear}
    baseBoneRef {boneName=root}
    componentAppliedLinearAccClamp=<1000 1000 1000>
    componentLinearAccScale=<0.2 0.4 0.3>
    overrideWorldGravity {z=-980}
    simSpaceSettings {
     maxAngularAcceleration=70
     maxAngularVelocity=200
     maxLinearAcceleration=100
     maxLinearVelocity=10
     velocityScaleZ=0.5
     worldAlpha=0.6
     externalLinearDragV=<0.1 0.1>
    }
   }
   n5 = RigidBody!(
    ComponentPose=@n4
    AlphaCurveName:Name=wings_dynamics_ctrl
    OverrideWorldGravity="0,0,-980"
    ExternalForce
    ComponentLinearAccScale=Get Component Linear Acc Scale() -> (Component Linear Acc Scale:Vector)
    ComponentLinearVelScale=Get Component Linear Vel Scale() -> (Component Linear Vel Scale:Vector)
   ) -> (Pose) {
    alphaInputType=Curve
    bOverrideWorldGravity=true
    overridePhysicsAsset=&/M_MED_BirdNestGreen_Pack_Physics_wings
    alphaBoolBlend {blendOption=Linear}
    componentAppliedLinearAccClamp=<3000 3000 900>
    componentLinearAccScale=<0.2 0.4 0.3>
    overrideWorldGravity {z=-980}
    simSpaceSettings {
     maxAngularAcceleration=800
     maxAngularVelocity=200
     maxLinearAcceleration=300
     maxLinearVelocity=10
     velocityScaleZ=0
     worldAlpha=1
     externalLinearDragV=<0.1 0.1>
    }
   }
   Output Pose(Result=Apply Additive(
     Base=Component To Local(ComponentPose=@n5)
     Additive="Sequence Player (additive) 'BirdNestGreen_IDLE_BACKBLING'"() {sequence=/BRCosmetics/Accessories/FORT_Backpacks/Backpack_M_MED_BirdNestGreen/Animation/BirdNestGreen_IDLE_BACKBLING}
    ) {alphaBoolBlend {blendOption=Linear}}) {layerGroup=None}
  ]}
 EventGraph {nodes [
   Event BlueprintUpdateAnimation()
   > n1 = Sequence!(execute) -> (then_0, then_1, then_2:Exec, then_3:Exec)
   > n2 = Branch(Condition=Get bIsSkyDiving())
   > Set Component Linear Acc Scale(Component Linear Acc Scale:Vector="0.2,0.3,0.3") -> (Output_Get:Vector)
   > Set Component Linear Acc Scale(Component Linear Acc Scale="0.1,0.1,0.1")
   Set Component Linear Vel Scale(execute=@n2.else, Component Linear Vel Scale:Vector="0.4,0.4,0.4") -> (Output_Get:Vector)
   > Set Component Linear Acc Scale(Component Linear Acc Scale="0.4,0.4,0.5")
   n3 = Gravity Override State Function DEPRECATED(
    execute=@n1.then_3
    CharacterPartAnimInstance=Self-Reference()
    GameTimeInSeconds=GetGameTimeInSeconds()
    GravityOverrideParams=Get Gravity Override Params()
   )
   > Set Gravity Override(Gravity Override=@n3)
   > Set FullSkyDiving(FullSkyDiving:bool=OR Boolean!(A=Get bIsSkyDiving(), B=Get bIsSkydiveFloating(), C:bool=Get bIsSkydiveDiveMode()) -> (ReturnValue)) -> (Output_Get:bool)
   n4 = Branch(execute=@n1.then_2, Condition=Get FullSkyDiving())
   > Set Component Linear Acc Scale_0(Component Linear Acc Scale_0:Vector="0.6,0.8,0.4") -> (Output_Get:Vector)
   > Set Component Linear Acc Scale_0(Component Linear Acc Scale_0="0.4,0.5,0.4")
   Set Component Linear Vel Scale_0(execute=@n4.else, Component Linear Vel Scale_0:Vector="0.18,0.4,0.4") -> (Output_Get:Vector)
   > Set Component Linear Acc Scale_0(Component Linear Acc Scale_0="0.3,0.3,0.4")
   n5 = Get bIsTacticalSprinting()
   n6 = Get bIsMoving2D()
   n7 = Branch(execute=@n1.then_1, Condition=AND Boolean(A=OR Boolean!(A=@n6, B=@n5) -> (ReturnValue), B=Get bIsOnGround() -> (bIsOnGround:bool)))
   > Set Component Linear Vel Scale(Component Linear Vel Scale="0.5,0.5,0.8")
   > Set Component Linear Acc Scale(Component Linear Acc Scale="0.5,0.5,0.6")
   Set Component Linear Acc Scale(execute=@n7.else, Component Linear Acc Scale="0.4,0.4,0.4")
   > Set Component Linear Acc Scale(Component Linear Acc Scale="0.4,0.4,0.7")
   TryGetPawnOwner()
  ]}
]
variables [
 Component Linear Acc Scale {type=Vector, default=<0.2 0.4 0.3>}
 Component Linear Acc Scale_0 {type=Vector, default=<0.18 0.4 0.3>}
 Component Linear Vel Scale {type=Vector}
 Component Linear Vel Scale_0 {type=Vector, default=<0.3 0.3 0.2>}
 FullSkyDiving {type=boolean}
 Gravity Override {type=Vector}
 Gravity Override Params {
  type=GravityOverrideParamsStruct
  default {
   skydive_DivingDown {gravityOverride=<-1000 -500 0>, windAmplitude=<50 200 100>}
   skydive_FloatingDown {gravityOverride=<0 0 500>, windFrequency=<1.5 7 13>, windOffset=<0 -300 300>}
   skydive_GoingUp {gravityOverride=<0 0 500>, windOffset=<0 -400 400>}
  }
 }
]
